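; Gameplay ability for the TidalCrane ToyFox boss that knocks back players when it unburrows, built on a parent launch ability.
; The launch event runs the parent launch with the given velocity, then applies a no-FX fall damage immunity effect to the ability owner.
; FindLauncher ignores its default and returns the effect causer from the context that granted the ability, so that causer acts as the launcher.
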(event EventDoLaunch
  (params (Vel :type Vector :default "0, 0, 0"))
  ;[0 0.15 0.0122 0.5] DoLaunch: Provide Fall damage immunity.
  (|CallParentFunction Vel)
  (Ability|ApplyGameplayEffectToOwner "/Game/Athena/Items/Parents/FallImmunity/GE_Athena_FallDamageImmunity_NoFX.GE_Athena_FallDamageImmunity_NoFX_C"))

(fn FindLauncher
  (params (DefaultLauncher :type Actor%20Object%20Reference))
  (return (Ability|EffectContext|GetEffectCauser (Ability|GetGrantedbyEffectContext))))
