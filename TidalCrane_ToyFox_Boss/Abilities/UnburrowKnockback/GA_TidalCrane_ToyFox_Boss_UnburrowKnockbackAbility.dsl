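; Gameplay ability for the TidalCrane ToyFox boss that knocks back nearby targets, apparently as it unburrows.
; On activation it disables boss collision, waits briefly, then runs a targeting preset task, rescheduling it every tick while active.
; Each pass applies the knockback GameplayEffect to targets and makes hit player pawns ignore boss collision for a hotfixable duration.
; Ending the ability clears the timer, ends the task and restores boss collision.
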(event Ability|EventActivateAbility
  ;[0.01165 0.15 0 0.5] ActivateAbility: Start targeting for knockback.
  (|CallParentFunction)
  (Fortnite|FortShippingLog self "ActivateAbility: Invoked.")
  (bind _asbga_tidal_crane_boss_toy_fox (Utilities|Casting|CastToBGA_TidalCrane_Boss_ToyFox (Ability|GetAvatarActorfromActorInfo))
    (:then
      (bind _output_get (Variables|Transient|SetBossActor _asbga_tidal_crane_boss_toy_fox))
      (Class|BGATidalCraneBossToyFox|SetBossCollisionEnable _output_get false)
      (Utilities|FlowControl|Delay 0.2)
      (CallFunction|BeginTargetingforKnockback))
    (:CastFailed
      (Ability|CancelAbility))))

(event Custom|BeginTargetingForKnockback
  ;[0.03047 0.15 0.0678 0.5] BeginTargetingForKnockback: Start continous targeting and application of knockback.
  (bind (_asynctaskproxy _targetdata) (Ability|Tasks|PerformTargetingAbilityTask "KnockbackTargetingPreset" (Variables|Settings|GetKnockbackTargetingPreset) false false true)
    (:then
      (Variables|Transient|SetAsyncTask_Targeting _asynctaskproxy))
    (:Targeted
      (if (Ability|GameplayAbility|IsActive 0)
        (bind _returnvalue (Utilities|Time|SetTimerforNextTickbyEvent (delegate Custom|BeginTargetingForKnockback)))
        (Variables|Transient|SetTargetingTImerHandle _returnvalue)
        (Utilities|FlowControl|Sequence
          (:then_0
            (for _array_element (Ability|TargetData|GetAllActorsfromTargetData _targetdata)
              (bind _asfort_player_pawn (Utilities|Casting|CastToFortPlayerPawn _array_element))
              (Collision|AddIgnoredActor _asfort_player_pawn (Variables|Transient|GetBossActor) (ScalableFloat|GetCachedScalableFloatValue (Variables|Hotfix|GetHotfix_IgnorePlayerDuration)))))
          (:then_1
            (Ability|ApplyGameplayEffectToTarget _targetdata (Variables|Settings|GetKnockbackGameplayEffectClass))))))))

(event Ability|EventOnEndAbility
  (params (bWasCancelled :type Boolean))
  ;[0.09375 0.06339 0 0.5] EndAbility: Clean up.
  (Utilities|FlowControl|Sequence
    (:then_0
      (Utilities|Time|ClearandInvalidateTimerbyHandle (Variables|Transient|GetTargetingTImerHandle))
      (bind _asynctask_targeting (Variables|Transient|GetAsyncTask_Targeting)
        (:then
          (GameplayTasks|EndTask _asynctask_targeting))
        (:else
          _))
      (bind _bossactor (Variables|Transient|GetBossActor)
        (:then
          (Class|BGATidalCraneBossToyFox|SetBossCollisionEnable _bossactor true))
        (:else)))
    (:then_1
      (|CallParentFunction bWasCancelled))))

(fn CanActivateAbility
  (params (ActorInfo :type Gameplay%20Ability%20Actor%20Info%20Structure) (Handle :type Gameplay%20Ability%20Spec%20Handle%20Structure))
  (return 0 true))
