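; Targeting task for the TidalCrane ToyFox boss's unburrow knockback that overrides where the targeting query originates.
; GetSourceLocation reads the source context from the targeting subsystem and casts it to the boss.
; It builds a hybrid point from the actor's X and Y and the Z of the mesh's spine_04 socket, so the height follows the animated body.
; A Sequence provides the parent implementation's location as a fallback.
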(fn GetSourceLocation
  (params (TargetingHandle :type Targeting%20Request%20Handle%20Structure%20%28by%20ref%29))
  (bind _returnvalue (Targeting|GetTargetingSubsystem TargetingHandle))
  (Utilities|FlowControl|Sequence
    (:then_0
      (Utilities|IsValid _returnvalue
        (:Is%20Valid
          (bind _asbga_tidal_crane_boss_toy_fox (Utilities|Casting|CastToBGA_TidalCrane_Boss_ToyFox (Utilities|Struct|BreakTargetingSourceContext (TargetingSystem|TargetingTypes|GetTargetingSourceContext _returnvalue TargetingHandle))))
          (bind _skeletalmeshcomponentbudgeted (Variables|/TidalCrane_Swarm_Boss/Gameplay/BGA_TidalCrane_Boss.BGA_TidalCrane_Boss_C|GetSkeletalMeshComponentBudgeted _asbga_tidal_crane_boss_toy_fox)
            (:then
              (bind (_returnvalue_x _returnvalue_y _returnvalue_z) (Math|Vector|BreakVector (Transformation|GetSocketLocation _skeletalmeshcomponentbudgeted "spine_04")))
              (bind (_returnvalue_x_2 _returnvalue_y_2) (Math|Vector|BreakVector (Transformation|GetActorLocation _asbga_tidal_crane_boss_toy_fox)))
              (return _returnvalue_x_2 _returnvalue_y_2 _returnvalue_z))
            (:else)))))
    (:then_1
      (bind _returnvalue_2 (|CallParentFunction TargetingHandle))
      (return _returnvalue_2))))
